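# Sound cue for the Butterfly pickaxe cosmetic's appear sound, routed to the melee weapon sound class.
# A Random node picks one of three equally weighted appear waves and plays it at a heavily reduced volume.
# The cue overrides attenuation for a short-range falloff and overrides concurrency to prevent new instances once the limit is reached.
# It primes on load and has its debug flag enabled.
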SoundCue /BRCosmetics/Sounds/PickAxes/Butterfly/Pickaxe_Butterfly_Appear_Cue
root=Random_0
nodes [
 {id=Random_0, children [WavePlayer_0,WavePlayer_1,WavePlayer_2], properties {weights [1,1,1]}}
 {id=WavePlayer_0, properties {soundWaveAssetPtr=&/Athena_Pickaxe_Butterfly_Appear_01}}
 {id=WavePlayer_1, properties {soundWaveAssetPtr=&/Athena_Pickaxe_Butterfly_Appear_02}}
 {id=WavePlayer_2, properties {soundWaveAssetPtr=&/Athena_Pickaxe_Butterfly_Appear_03}}
]
properties {
 bDebug=true
 bOverrideAttenuation=true
 bOverrideConcurrency=true
 bPrimeOnLoad=true
 maxDistance=1000
 soundClassObject=&Fort_WeaponMelee
 volumeMultiplier=0.15
 attenuationOverrides {falloffDistance=500, attenuationShapeExtents {x=500}}
 concurrencyOverrides {resolutionRule=PreventNew}
}
